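# Physics asset for the CrispSeasonOffhand pickaxe, a melee weapon cosmetic.
# It defines a single capsule body on the Melee_TwoHanded_Handle bone, offset and rotated from the bone origin.
# The capsule's radius nearly equals its length, giving a short, rounded collision volume, and the asset has no constraints.
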PhysicsAsset /BRCosmetics/Weapons/FORT_Melee/Pickaxe_CrispSeasonOffhand/Meshes/CrispSeasonOffhand_PhysicsAsset
bodies {Melee_TwoHanded_Handle [{shapeType=Capsule, length=21.8123, radius=21.0446, center=<1.0103 -18.5258 1.8852>, rotation=<rot 69.2957 88.8677 178.9408>}]}
constraints []
